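# Physics asset for the SoloSnooze pickaxe cosmetic, defining capsule bodies and joint limits for a dangling or jiggling three-bone dynamic section.
# The dyn_SoloSnooze 01 to 03 bodies form a chain from a tiny base_SoloSnooze capsule, with wider swing at the root joint and tight twist throughout.
# An unconstrained thigh_r capsule is also included, likely as a collider for the wielder's right leg.
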PhysicsAsset /BRCosmetics/Weapons/FORT_Melee/Pickaxe_SoloSnooze/Meshes/SoloSnooze_Axe_Physics
bodies {
 base_SoloSnooze [{shapeType=Capsule, length=0.505, radius=0.505, shapeName=base_SleepyCloud_capsule, rotation=<rot 0 -90 0>}]
 dyn_SoloSnooze_01 [{shapeType=Capsule, length=8.5, radius=5.52, center=<8.4332 0 -2.285>, rotation=<rot 49.9998 -0.0002 -0.0001>}]
 dyn_SoloSnooze_02 [{shapeType=Capsule, length=8.5, radius=5.95, center=<5.404 0 0>}]
 dyn_SoloSnooze_03 [{shapeType=Capsule, length=8.5, radius=7, center=<7.0568 0 0>}]
 thigh_r [{shapeType=Capsule, length=34.6616, radius=9.5479, shapeName=thigh_r_capsule, center=<22.1907 0.6066 0.9958>, rotation=<rot 90 0 0>}]
}
constraints [
 {bone1Name=dyn_SoloSnooze_02, bone2Name=dyn_SoloSnooze_01, swing1LimitDegrees=20, swing2LimitDegrees=15, twistLimitDegrees=4}
 {bone1Name=dyn_SoloSnooze_03, bone2Name=dyn_SoloSnooze_02, swing1LimitDegrees=25, swing2LimitDegrees=15, twistLimitDegrees=4}
 {bone1Name=dyn_SoloSnooze_01, bone2Name=base_SoloSnooze, swing1LimitDegrees=45, swing2LimitDegrees=50, twistLimitDegrees=4}
]
