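# Physics asset for the SoloSnooze pickaxe, defining capsule bodies and joint limits for secondary motion of its dangling, limbed figure.
# A three-segment spine chains from base_SoloSnooze, with ears and both shoulder-elbow arms branching from the top segment and a left leg-knee chain from the bottom.
# Joints have tight swing and twist limits, with twist locked on the leg.
# A large thigh_r capsule has no constraint, apparently serving as a collider.
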PhysicsAsset /BRCosmetics/Weapons/FORT_Melee/Pickaxe_SoloSnooze/Meshes/SoloSnooze_Axe_PhysicsFE
bodies {
 base_SoloSnooze [{shapeType=Capsule, length=0.505, radius=0.505, shapeName=base_SleepyCloud_capsule, rotation=<rot 0 -90 0>}]
 dyn_SoloSnooze_01 [{shapeType=Capsule, length=8.5, radius=5.7, center=<8.6616 0 -2.4767>, rotation=<rot 49.9998 -0.0002 -0.0001>}]
 dyn_SoloSnooze_02 [{shapeType=Capsule, length=8.5, radius=5.7, center=<5.404 0 -0.5132>}]
 dyn_SoloSnooze_03 [{shapeType=Capsule, length=8.5, radius=7, center=<7.0568 0 0>}]
 dyn_ear_l [{
   shapeType=Capsule
   length=2.8744
   radius=1.834
   shapeName=dyn_ear_l_capsule
   center=<1.7294 0.5258 0.1041>
   rotation=<rot 0 15.0614 -82.0409>
  }]
 dyn_ear_r [{
   shapeType=Capsule
   length=2.8774
   radius=1.9208
   shapeName=dyn_ear_r_capsule
   center=<1.5843 0.0119 -0.0991>
   rotation=<rot 0 3.8689 -51.1358>
  }]
 dyn_elbow_l [{shapeType=Capsule, length=1, radius=2, shapeName=dyn_elbow_l_capsule, center=<1.9876 -0.1865 -0.2982>, rotation=<rot 90 0 0>}]
 dyn_elbow_r [{shapeType=Capsule, length=1, radius=2, shapeName=dyn_elbow_r_capsule, center=<2.5707 0.1412 -0.1637>, rotation=<rot 90 0 0>}]
 dyn_knee_l [{shapeType=Capsule, length=3.5, radius=2, shapeName=dyn_knee_l_capsule, center=<2.6674 0.2496 -0.267>, rotation=<rot 90 0 0>}]
 dyn_leg_l [{shapeType=Capsule, length=3.5, radius=2, shapeName=dyn_leg_l_capsule, center=<3.4002 0.1838 -0.0204>, rotation=<rot 90 0 0>}]
 dyn_shoulder_l [{shapeType=Capsule, length=3, radius=2, shapeName=dyn_shoulder_l_capsule, center=<1.4303 0.4031 -0.4433>, rotation=<rot 90 0 0>}]
 dyn_shoulder_r [{shapeType=Capsule, length=2, radius=2, shapeName=dyn_shoulder_r_capsule, center=<1.6774 -0.0274 0.2663>, rotation=<rot 90 0 0>}]
 thigh_r [{shapeType=Capsule, length=34.6616, radius=9.5479, shapeName=thigh_r_capsule, center=<22.1907 0.6066 0.9958>, rotation=<rot 90 0 0>}]
}
constraints [
 {bone1Name=dyn_SoloSnooze_02, bone2Name=dyn_SoloSnooze_01, swing1LimitDegrees=15, swing2LimitDegrees=8, twistLimitDegrees=4}
 {bone1Name=dyn_SoloSnooze_03, bone2Name=dyn_SoloSnooze_02, swing1LimitDegrees=15, swing2LimitDegrees=8, twistLimitDegrees=4}
 {bone1Name=dyn_SoloSnooze_01, bone2Name=base_SoloSnooze, swing1LimitDegrees=25, swing2LimitDegrees=30, twistLimitDegrees=4}
 {bone1Name=dyn_ear_r, bone2Name=dyn_SoloSnooze_03, swing1LimitDegrees=8, swing2LimitDegrees=25, twistLimitDegrees=4}
 {bone1Name=dyn_ear_l, bone2Name=dyn_SoloSnooze_03, swing1LimitDegrees=8, swing2LimitDegrees=25, twistLimitDegrees=4}
 {bone1Name=dyn_shoulder_r, bone2Name=dyn_SoloSnooze_03, swing1LimitDegrees=15, swing2LimitDegrees=8, twistLimitDegrees=4}
 {bone1Name=dyn_elbow_r, bone2Name=dyn_shoulder_r, swing1LimitDegrees=15, swing2LimitDegrees=8, twistLimitDegrees=4}
 {bone1Name=dyn_shoulder_l, bone2Name=dyn_SoloSnooze_03, swing1LimitDegrees=15, swing2LimitDegrees=8, twistLimitDegrees=4}
 {bone1Name=dyn_elbow_l, bone2Name=dyn_shoulder_l, swing1LimitDegrees=15, swing2LimitDegrees=8, twistLimitDegrees=4}
 {bone1Name=dyn_leg_l, bone2Name=dyn_SoloSnooze_01, swing1LimitDegrees=15, swing2LimitDegrees=8, twistLimitDegrees=45, twistMotion=Locked}
 {bone1Name=dyn_knee_l, bone2Name=dyn_leg_l, swing1LimitDegrees=15, swing2LimitDegrees=8, twistLimitDegrees=45, twistMotion=Locked}
]
